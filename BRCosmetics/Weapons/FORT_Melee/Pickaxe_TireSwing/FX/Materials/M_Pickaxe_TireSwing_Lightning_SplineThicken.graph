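# Translucent unlit material for the TireSwing pickaxe's lightning, drawn on Niagara mesh particles and colored by particle color.
# World position offset widens strips toward a camera-facing spline, nudges them toward the camera, pulls ends toward local center, and adds noise displacement strongest mid-length.
# Panning noise modulates the edge-falloff exponent for opacity, multiplied by particle alpha, tip fade and platform-switched depth fade.
# Dynamic parameters drive width, noise influence and exponent.
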Material /BRCosmetics/Weapons/FORT_Melee/Pickaxe_TireSwing/FX/Materials/M_Pickaxe_TireSwing_Lightning_SplineThicken
graph {
 properties {
  bAutomaticallySetUsageInEditor=false
  bUsedWithNiagaraMeshParticles=true
  blendMode=BLEND_Translucent
  pixelDepthOffsetMode=PDOM_Legacy
  shadingModel=MSM_Unlit
  usedShadingModels=Unlit
 }
 nodes [
  Add(A=Add(A=NamedRerouteUsage(), B=NamedRerouteUsage()), B=NamedRerouteUsage())
  n1 = ParticleColor()
  MaterialFunctionCall(Emissive=@n1.RGB) {materialFunction=&MF_FX_EmissiveAdjustments}
  n2 = ConstantBiasScale(None=ComponentMask(None=TextureCoordinate()) {r=true}) {bias=-0.5, scale=2}
  n3 = DynamicParameter() {defaultValue=<rgba 1 6 1 1>, paramNames [Noise Influence, Noise Exponent, Spline Width, Param4]}
  n4 = Constant()
  n5 = Multiply(A=ComponentMask(None=TextureCoordinate()) {g=true}, B=ScalarParameter() {defaultValue=2, parameterName=Noise UV Scale})
  n6 = ScalarParameter() {defaultValue=0.5, parameterName=Noise UV Panner Speed}
  n7 = Constant()
  n8 = TextureSample(UVs=Panner(Coordinate=AppendVector(A=@n4, B=@n5), Speed=AppendVector(A=@n6, B=@n7)) {speedX=1}) {
   samplerSource=SSM_Wrap_WorldGroupSettings
   samplerType=SAMPLERTYPE_LinearGrayscale
   texture=/BRCosmetics/Weapons/FORT_Melee/Pickaxe_Parfait_Female/FX/Textures/T_TrifleParfait_LightningNoise
  }
  n9 = Saturate(None=Multiply(A=@n8.R, B=@n3.Noise Influence))
  n10 = Power(Base=OneMinus(None=Abs(None=@n2)), Exp=Divide(A=@n3.Noise Exponent, B=OneMinus(None=@n9)))
  Multiply(A=Multiply(A=Multiply(A=@n1.A, B=@n10), B=NamedRerouteUsage()), B=NamedRerouteUsage())
  n11 = Subtract(A=Add(A=WorldPosition().XYZ, B=NamedRerouteUsage()), B=CameraPositionWS())
  n12 = Transform(None=Constant3Vector() {constant=<rgba 0 -1 0 0>})
  n13 = Normalize(VectorInput=CrossProduct(A=Normalize(VectorInput=@n11), B=Normalize(VectorInput=@n12)))
  n14 = Subtract(A=ComponentMask(None=TextureCoordinate()) {r=true}) {constB=0.5}
  NamedRerouteDeclaration(None=Multiply(
    A=Multiply(A=@n13, B=@n14)
    B=DynamicParameter() {defaultValue=<rgba 1 6 1 1>, paramNames [Noise Influence, Noise Exponent, Spline Width, Param4]}.Spline Width
   )) {name=Spline Thicken WPO, nodeColor=<rgba 0 1 0.5547 1>}
  n15 = Multiply(A=LocalPosition().XY) {constB=-1}
  n16 = Constant()
  NamedRerouteDeclaration(None=Transform(None=AppendVector(A=@n15, B=@n16)) {transformSourceType=TRANSFORMSOURCE_Local}) {name=Move to Center, nodeColor=<rgba 1 0.8672 0 1>}
  n17 = NamedRerouteUsage()
  n18 = NamedRerouteUsage()
  n19 = Subtract(A=CameraPositionWS(), B=Add(A=Add(A=@n17, B=@n18), B=WorldPosition().XYZ))
  NamedRerouteDeclaration(None=Multiply(A=Normalize(VectorInput=@n19), B=ScalarParameter() {defaultValue=2, parameterName=Camera Offset})) {name=Camera Offset, nodeColor=<rgba 1 0.2344 0 1>}
  n20 = NamedRerouteUsage()
  n21 = ConstantBiasScale(None=ComponentMask(None=TextureCoordinate()) {g=true}) {bias=-0.5, scale=2}
  n22 = DotProduct(A=@n21, B=@n21)
  n23 = Multiply(
   A=Power(Base=@n22, Exp=ScalarParameter() {defaultValue=1, parameterName=Move to Center Exponent})
   B=ScalarParameter() {defaultValue=0.75, parameterName=Move to Center Amount}
  )
  n24 = Normalize(VectorInput=@n20)
  n25 = ParticleRandom()
  n26 = Multiply(A=ComponentMask(None=TextureCoordinate()) {g=true}, B=ScalarParameter() {defaultValue=0.5, parameterName=Displacement Noise UV Scale})
  n27 = ScalarParameter() {defaultValue=0.5, parameterName=Displacement Noise UV Panner Speed}
  n28 = Constant()
  n29 = TextureSample(UVs=Panner(Coordinate=AppendVector(A=@n25, B=@n26), Speed=AppendVector(A=@n27, B=@n28)) {speedX=1}) {
   samplerSource=SSM_Wrap_WorldGroupSettings
   samplerType=SAMPLERTYPE_LinearGrayscale
   texture=&T_TrifleParfait_LightningNoise
  }
  n30 = Multiply(
   A=OneMinus(None=@n22)
   B=Multiply(A=Subtract(A=@n29.R) {constB=0.5}, B=ScalarParameter() {defaultValue=3, parameterName=Displacement Noise Intensity})
  )
  NamedRerouteDeclaration(None=Add(A=Multiply(A=@n20, B=@n23), B=Multiply(A=@n24, B=@n30))) {name=Position Offset, nodeColor=<rgba 0 1 0.2266 1>}
  n31 = DepthFade(FadeDistance=ScalarParameter() {defaultValue=3, parameterName=Depth Fade Distance})
  n32 = Constant() {r=1}
  NamedRerouteDeclaration(None=MaterialFunctionCall("Xb1,PS4"=@n31, "PC-High&&Epic,XSX,PS5"=@n31, "PC-Perf,Android"=@n32, Switch=@n32, "PC-Low(NotPerf)&&Medium"=@n32) {materialFunction=&PlatformSwitch_1D}.Result) {name=Depth Fade, nodeColor=<rgba 0 0.6484 1 1>}
  n33 = ComponentMask(None=TextureCoordinate()) {g=true}
  n34 = OneMinus(None=Abs(None=ConstantBiasScale(None=@n33) {bias=-0.5, scale=2}))
  n35 = ScalarParameter() {defaultValue=0.3, parameterName=Tips Fade Size}
  NamedRerouteDeclaration(None=Saturate(None=Divide(A=@n34, B=@n35))) {name=Tips Fade, nodeColor=<rgba 0 1 0.9063 1>}
 ]
}
