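# Unlit, two-sided additive overlay material for Niagara mesh particles on the TireSwing pickaxe tube effect.
# Ring gradients for sides and caps are blended with three averaged, wave-distorted, panning cellular noise samples, then tinted by particle color.
# Named reroutes supply a normal-based cap/side mask, axis-aligned fresnel, masked noise and depth-based opacity.
# Depth testing is disabled, with the opacity term fading pixels by their distance behind scene depth.
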Material /BRCosmetics/Weapons/FORT_Melee/Pickaxe_TireSwing/FX/Materials/M_Pickaxe_TireSwing_TubeOverlay
graph {
 properties {
  bAutomaticallySetUsageInEditor=false
  bDisableDepthTest=true
  bUsedWithNiagaraMeshParticles=true
  blendMode=BLEND_Additive
  pixelDepthOffsetMode=PDOM_Legacy
  shadingModel=MSM_Unlit
  twoSided=true
  usedShadingModels=Unlit
 }
 nodes [
  n1 = TextureCoordinate()
  n2 = Abs(None=ConstantBiasScale(None=ComponentMask(None=@n1) {g=true}) {bias=-0.5, scale=2})
  n3 = Saturate(None=Divide(A=OneMinus(None=@n2), B=ScalarParameter() {defaultValue=0.3, parameterName=Ring Thickness Sides}))
  n4 = MaterialFunctionCall!(
   Blend=MaterialFunctionCall(Blend=OneMinus(None=@n3), Base=NamedRerouteUsage()) {materialFunction=/Engine/Functions/Engine_MaterialFunctions03/Blends/Blend_Screen}.Result
   Base=NamedRerouteUsage()
  ) {materialFunction=&Blend_Screen}
  n5 = Constant() {r=0.5}
  n6 = Distance(A=TextureCoordinate(), B=@n5)
  n7 = OneMinus(None=Saturate(None=Divide(A=@n6, B=@n5)))
  n8 = ScalarParameter() {defaultValue=0.4, parameterName=Ring Thickness Caps}
  n9 = OneMinus(None=Saturate(None=Divide(A=@n7, B=@n8)))
  n10 = NamedRerouteUsage()
  n11 = ScalarParameter() {defaultValue=3, parameterName=Noise UV Wave Frequency}
  n12 = ScalarParameter() {defaultValue=0.025, parameterName=Noise UV Wave Amplitude}
  n13 = MaterialFunctionCall!(
   UV=MaterialFunctionCall!(
    UV=TextureCoordinate() {vTiling=0.5}
    U Frequency=@n11
    V Frequency=@n11
    U Amplitude=@n12
    V Amplitude=@n12
    U Speed
    V Speed
    U Time
    V Time
   ) {materialFunction=/Game/Effects/Fort_Effects/MaterialFunctions/MF_WaveUV}.Wave UVs
   Rotation
   Speed=ScalarParameter() {defaultValue=0.75, parameterName=Noise Pan Speed}
   Time
   Rotate UVs=StaticBool()
   Fractional Part
  ) {materialFunction=/Game/Accessories/FORT_Backpacks/Backpack_M_MED_PitGlass/FX/Materials/MF_3UVPanner}
  n14 = OneMinus(None=TextureSample(UVs=@n13.UV1) {samplerType=SAMPLERTYPE_Masks, texture=/Game/Global/Textures/Default/Noise/cellularNoise}.B)
  n15 = OneMinus(None=TextureSample(UVs=@n13.UV2) {samplerType=SAMPLERTYPE_Masks, texture=&cellularNoise}.B)
  n16 = TextureSample(UVs=@n13.UV3) {samplerType=SAMPLERTYPE_Masks, texture=&cellularNoise}
  n17 = Divide(A=Add(A=Add(A=@n14, B=@n15), B=OneMinus(None=@n16.B))) {constB=3}
  n18 = MaterialFunctionCall!(
   Base=Power(Base=LinearInterpolate(A=@n4.Result, B=@n9, Alpha=@n10), Exp=ScalarParameter() {defaultValue=2, parameterName=Ring Exponent})
   Blend=@n17
  ) {materialFunction=/Game/Global/Materials/Character_Parent_2023/Anime_Parent_2023/MaterialFunctions/MF_BlendOverlay}
  n19 = ScalarParameter() {defaultValue=0.1, parameterName=Background Intensity}
  MaterialFunctionCall(Emissive=Multiply(
    A=ParticleColor().RGB
    B=MaterialFunctionCall(Blend=@n18.Result, Base=@n19) {materialFunction=&Blend_Screen}.Result
   )) {materialFunction=&MF_FX_EmissiveAdjustments}
  n20 = CameraPositionWS()
  n21 = WorldPosition()
  Multiply(A=Normalize(VectorInput=Subtract(A=@n20, B=@n21.XYZ)), B=ScalarParameter() {defaultValue=-1, parameterName=Camera Offset})
  n22 = PixelDepth()
  n23 = SceneDepth()
  n24 = Divide(
   A=Subtract(A=Subtract(A=@n22.R, B=@n23.R), B=ScalarParameter() {defaultValue=2, parameterName=Depth Visibility Offset})
   B=ScalarParameter() {defaultValue=6, parameterName=Depth Visibility Size}
  )
  NamedRerouteDeclaration(None=OneMinus(None=Saturate(None=@n24))) {name=Depth Based Opacity, nodeColor=<rgba 0.3359 1 0 1>}
  n25 = ScalarParameter() {defaultValue=0.666, parameterName=Fresnel Exponent}
  n26 = MaterialFunctionCall(Axis=Transform(None=Constant3Vector() {constant=<rgba 0 0 1 0>}) {transformSourceType=TRANSFORMSOURCE_Instance}, Exponent=@n25) {materialFunction=&AxisAlignedFresnel}
  NamedRerouteDeclaration(None=OneMinus(None=Power(Base=@n26.Fresnel, Exp=@n25))) {name=Axis Aligned Fresnel, nodeColor=<rgba 1 0 0.6562 1>}
  n27 = Round(None=ComponentMask(None=PreSkinnedNormal()) {b=true})
  NamedRerouteDeclaration(None=VertexInterpolator(VS=Abs(None=@n27)).PS) {name=Cap Sides Mask, nodeColor=<rgba 0.3828 0 1 1>}
  n28 = Power(Base=@n17, Exp=ScalarParameter() {defaultValue=2, parameterName=Noise Exponent})
  n29 = ScalarParameter() {defaultValue=2, parameterName=Noise Intensity}
  n30 = ConstantBiasScale(None=ComponentMask(None=TextureCoordinate()) {g=true}) {bias=-0.5, scale=2}
  n31 = Divide(A=OneMinus(None=Abs(None=@n30)), B=ScalarParameter() {defaultValue=1, parameterName=Noise Mask Size})
  NamedRerouteDeclaration(None=Multiply(A=Multiply(A=@n28, B=@n29), B=Saturate(None=@n31))) {name=Noise, nodeColor=<rgba 0 1 0.7656 1>}
  NamedRerouteUsage()
  VectorParameter() {parameterName=Color, defaultValue=<rgba 0.95 0.2 2 0>}
 ]
}
