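# Translucent unlit Niagara mesh particle material for the Tire Swing pickaxe's tube overlay effect, drawing a soft, view-facing strip along a spline-like mesh.
# World position offset widens the mesh across U, perpendicular to the view and its local Y axis, then nudges it toward the camera.
# Particle color feeds emissive adjustments, while particle alpha multiplies a centered U falloff whose exponent is modulated by panning lightning noise, varying the soft edge along V.
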Material /BRCosmetics/Weapons/FORT_Melee/Pickaxe_TireSwing/FX/Materials/M_Pickaxe_TireSwing_TubeOverlay_SplineThicken
graph {
 properties {
  bAutomaticallySetUsageInEditor=false
  bUsedWithNiagaraMeshParticles=true
  blendMode=BLEND_Translucent
  pixelDepthOffsetMode=PDOM_Legacy
  shadingModel=MSM_Unlit
  usedShadingModels=Unlit
 }
 nodes [
  Add(A=NamedRerouteUsage(), B=NamedRerouteUsage())
  n1 = ParticleColor()
  MaterialFunctionCall(Emissive=@n1.RGB) {materialFunction=&MF_FX_EmissiveAdjustments}
  n2 = TextureCoordinate()
  n3 = Abs(None=ConstantBiasScale(None=ComponentMask(None=@n2) {r=true}) {bias=-0.5, scale=2})
  n4 = ScalarParameter() {defaultValue=1.5, parameterName=Noise Exponent}
  n5 = Constant()
  n6 = Multiply(A=ComponentMask(None=TextureCoordinate()) {g=true}, B=ScalarParameter() {defaultValue=1, parameterName=Noise UV Scale})
  n7 = ScalarParameter() {defaultValue=0.5, parameterName=Noise UV Panner Speed}
  n8 = Constant()
  n9 = TextureSample(UVs=Panner(Coordinate=AppendVector(A=@n5, B=@n6), Speed=AppendVector(A=@n7, B=@n8)) {speedX=1}) {
   samplerSource=SSM_Wrap_WorldGroupSettings
   samplerType=SAMPLERTYPE_LinearGrayscale
   texture=/BRCosmetics/Weapons/FORT_Melee/Pickaxe_Parfait_Female/FX/Textures/T_TrifleParfait_LightningNoise
  }
  n10 = ScalarParameter() {defaultValue=1, parameterName=Noise Influence}
  n11 = OneMinus(None=Saturate(None=Multiply(A=@n9.R, B=@n10)))
  Multiply(A=@n1.A, B=Power(Base=OneMinus(None=@n3), Exp=Divide(A=@n4, B=@n11)))
  n12 = Subtract(A=WorldPosition().XYZ, B=CameraPositionWS())
  n13 = Transform(None=Constant3Vector() {constant=<rgba 0 -1 0 0>})
  n14 = Normalize(VectorInput=CrossProduct(A=Normalize(VectorInput=@n12), B=Normalize(VectorInput=@n13)))
  n15 = Subtract(A=ComponentMask(None=TextureCoordinate()) {r=true}) {constB=0.5}
  NamedRerouteDeclaration(None=Multiply(A=Multiply(A=@n14, B=@n15), B=ScalarParameter() {defaultValue=1.25, parameterName=Width})) {name=Spline Thicken WPO, nodeColor=<rgba 0 1 0.5547 1>}
  n16 = Subtract(A=CameraPositionWS(), B=Add(A=NamedRerouteUsage(), B=WorldPosition().XYZ))
  NamedRerouteDeclaration(None=Multiply(A=Normalize(VectorInput=@n16), B=ScalarParameter() {defaultValue=2, parameterName=Camera Offset})) {name=Camera Offset, nodeColor=<rgba 1 0.2344 0 1>}
 ]
}
